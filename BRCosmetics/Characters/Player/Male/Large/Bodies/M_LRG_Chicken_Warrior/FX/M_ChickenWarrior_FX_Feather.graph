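# Masked, two-sided lit material for Chicken Warrior feather effects on Niagara sprite and mesh particles, reusing the Poultry prop textures.
# A hard step on the masked UV coordinate, U by default, isolates one strip of the diffuse texture.
# That strip is desaturated and tinted by particle color, while the rest keeps its original color.
# The Poultry normal and mask textures are also sampled.
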Material /BRCosmetics/Characters/Player/Male/Large/Bodies/M_LRG_Chicken_Warrior/FX/M_ChickenWarrior_FX_Feather
graph {
 properties {
  bUsedWithNiagaraMeshParticles=true
  bUsedWithNiagaraSprites=true
  blendMode=BLEND_Masked
  pixelDepthOffsetMode=PDOM_Legacy
  twoSided=true
  usedShadingModels=Default Lit
 }
 nodes [
  n1 = OneMinus(None=StaticComponentMaskParameter(None=TextureCoordinate()) {defaultR=true})
  n2 = Constant() {r=0.68}
  n3 = Clamp(None=Ceil(None=Subtract(A=@n1, B=@n2)))
  n4 = TextureSample() {texture=/Game/Environments/Props/Poultry/Textures/T_Poultry_D}
  n5 = Multiply(A=@n3, B=@n4.RGB)
  Add(A=Multiply(A=ParticleColor().RGB, B=Desaturation(None=@n5)), B=Multiply(A=OneMinus(None=@n3), B=@n4.RGB))
  TextureSample() {samplerType=SAMPLERTYPE_Normal, texture=/Game/Environments/Props/Poultry/Textures/T_Poultry_N}
  TextureSample() {samplerType=SAMPLERTYPE_Masks, texture=/Game/Environments/Props/Poultry/Textures/T_Poultry_S}
 ]
}
