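# Niagara effect that drips slurp sprites from the Chicken Warrior character's skeletal mesh.
# Idle_Drips spawns at a random rate scaled by running speed, with a separate front-end preview rate; Jump_Drips spawns from the Land parameter.
# Both pick random bones or sockets with per-index offsets, inherit velocity, then stretch under ramping gravity with animated material parameters.
# Skydiving suppresses gameplay spawning, and being airborne zeroes gameplay particle lifetime.
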NiagaraSystem /BRCosmetics/Characters/Player/Male/Large/Bodies/M_LRG_Chicken_Warrior/FX/NS_ChickenWarrior_Slurp_Idle
dependencies {
 usedDataInterfaces [
  &NiagaraDataInterfaceArrayFloat
  &NiagaraDataInterfaceArrayFloat3
  &NiagaraDataInterfaceCurve
  &NiagaraDataInterfaceVector2DCurve
 ]
 usedDynamicInputs []
 usedModules [&SystemState]
}
emitters {
 Idle_Drips {
  emitterUpdateScript [
   &EmitterState(
    Life Cycle Mode=::NewEnumerator1
    Loop Delay=&RandomRangeFloat(Minimum=0, Maximum=0)
    Loop Duration=&RandomRangeFloat(Minimum=0.1, Maximum=0.3)
    Recalculate Duration Each Loop=true
    UseLoopDelay=true
   )
   &SpawnRate(
    SpawnRate=&FN_FrontEndFloatModifier(
     Gameplay=&Multiply_Float(
      A=&Lerp_Float(
       A=&RandomRangeFloat(Minimum=-2, Maximum=3, Evaluation Type=::NewEnumerator1)
       B=&RandomRangeFloat(Minimum=2, Maximum=6, Evaluation Type=::NewEnumerator1)
       Alpha=&ClampFloat(Float=$User.runningSpeed)
      )
      B=&MakeCustomFloatFromBool(Bool=$User.IsSkydiving, True Float=0, False Float=1)
     )
     Front End=&RandomRangeFloat(Minimum=-3, Maximum=7, Evaluation Type=::NewEnumerator1)
    )
    Use Spawn Probability=false
   ) {enabled=false}
   &SpawnRate(
    SpawnRate=&MakeCustomFloatFromBool(
     Bool=$User.bIsFrontEndPreview
     True Float=&RandomRangeFloat(Minimum=-3, Maximum=7, Evaluation Type=::NewEnumerator1)
     False Float=&Multiply_Float(
      A=&Lerp_Float(
       A=&RandomRangeFloat(Minimum=-2, Maximum=3, Evaluation Type=::NewEnumerator1)
       B=&RandomRangeFloat(Minimum=2, Maximum=6, Evaluation Type=::NewEnumerator1)
       Alpha=&ClampFloat(Float=$User.runningSpeed)
      )
      B=&MakeCustomFloatFromBool(Bool=$User.IsSkydiving, True Float=0, False Float=1)
     )
    )
    Use Spawn Probability=false
   ) {module=SpawnRate001}
  ]
  particleSpawnScript [
   /SetVariables(Particles.WhichSocket:NiagaraInt32=&RandomRangeInt(Minimum=0, Maximum=3)) {isSetParameters=true, module=SetVariables_#1}
   /Niagara/Modules/Spawn/Initialization/InitializeParticle(
    Lifetime=&FN_FrontEndFloatModifier(Gameplay=&MakeCustomFloatFromBool(Bool=$User.IsOnGround, True Float=1, False Float=0), Front End=1)
    Module.Write Color=true
    Module.Write Lifetime=true
    Module.Write Mass=false
    Module.Write Material Random=true
    Module.Write Position=false
    Module.Write Scale=false
    Module.Write SpriteRotation=false
    Module.Write SpriteSize=true
    Position=&SimulationPosition()
    Sprite Size=<4 10>
   )
   &SkeletalMeshLocation(
    Bone / Socket Index=$Particles.WhichSocket
    Bone Sampling Mode=::NewEnumerator4
    Inherit Velocity=true
    "Inherit Velocity (Scale)"=-0.25
    MaskSpawnGroup=false
    Sampled Position Offset=&Add_Vector(
     A=&SelectVectorFromArray(
      Direct Array Index=$Particles.WhichSocket
      Vector Selection Array=DataInterface() {propertyValues {FloatData [<0 0 -2>, <0 0 -5>, <0 0 -5>]}}
     )
     B=&RandomVector(Vector Scale=7)
    )
    Sampled Position Offset Coordinate Space:ENiagaraMeshTransforms=::NewEnumerator1
    Skeletal Mesh=$User.SkeletalMesh
    VelocitySpeedLimit=false
   )
   &InheritVelocity(Module.Limit Speed=true, Velocity Limit=200, Velocity Scale=&VectorFromFloat(Value=0.1))
   &CameraOffset(Camera Offset Amount=15)
  ]
  particleUpdateScript [
   &ParticleState()
   &ScaleSpriteSize(Scale Factor=&Vector2DFromCurve(
     Vector2Curve=DataInterface() {propertyValues {
       XCurve {keys [
         <at 0 1.15 RCIM_Cubic>
         {arriveTangent=-0.8581, interpMode=RCIM_Cubic, leaveTangent=-0.8581, time=0.2835, value=1.0208}
         {arriveTangent=-0.0291, interpMode=RCIM_Cubic, leaveTangent=-0.0291, time=0.8741, value=0.4}
         <at 1 1>
        ]}
       YCurve {keys [
         <at 0 0 RCIM_Cubic>
         {arriveTangent=2.8602, interpMode=RCIM_Cubic, leaveTangent=2.8602, time=0.2835, value=0.8693}
         {arriveTangent=1.2989, interpMode=RCIM_Cubic, leaveTangent=1.2989, time=0.8741, value=2.5}
         <at 1 1.8>
        ]}
      }}
     ScaleCurve=<1 1>
    ))
   &GravityForce(Gravity=&Multiply_VectorByFloat(
     Vector=<1 1 -500>
     Float=&FloatFromCurve(FloatCurve=DataInterface() {propertyValues {Curve {keys [{arriveTangent=1.8039, interpMode=RCIM_Cubic, leaveTangent=1.8039, tangentMode=RCTM_User, time=0, value=0}, <at 1 1>]}}})
    ))
   &SolveForcesAndVelocity(Clamp Velocity=false, Limit Acceleration=false)
   &DynamicMaterialParameters(
    Index 0 Param 1=0.8
    Index 0 Param 2=&FloatFromCurve(FloatCurve=DataInterface() {propertyValues {Curve {keys [{arriveTangent=-1.8884, interpMode=RCIM_Cubic, leaveTangent=-1.8884, tangentMode=RCTM_User, time=0, value=1}, <at 1 0.26>]}}})
    Index 0 Param 3=1
    Index 0 Param 4=1
    Module.Param0WriteEnabled=true
    Module.Param1WriteEnabled=true
    Module.Param2WriteEnabled=true
    Module.Param3WriteEnabled=true
   )
   &ScaleColor(
    Scale Alpha=&FloatFromCurve(FloatCurve=DataInterface() {propertyValues {Curve {keys [<at 0.5 1 RCIM_Cubic>,<at 1 0>]}}})
    ScaleA=true
    ScaleRGB=true
   )
  ]
  rendererClasses [&NiagaraSpriteRendererProperties]
  renderers [{rendererClass=&NiagaraSpriteRendererProperties}]
 }
 Jump_Drips {
  emitterUpdateScript [
   &EmitterState(Life Cycle Mode=::NewEnumerator1, Loop Duration=0.1, Recalculate Duration Each Loop=true, UseLoopDelay=true)
   &SpawnRate(
    Spawn Probability=&MakeCustomFloatFromBool(Bool=$User.IsSkydiving, True Float=0, False Float=1)
    SpawnRate=&Multiply_Float(A=25, B=$User.Land)
    Use Spawn Probability=true
   )
   /SetVariables(Emitter.PlayerVelocityVector:Vector3f=$Engine.Owner.Velocity) {isSetParameters=true}
   /SetVariables(Emitter.PlayerSpeed:NiagaraFloat=&VectorLength(Vector=$Emitter.PlayerVelocityVector)) {isSetParameters=true}
  ]
  particleSpawnScript [
   /SetVariables(Particles.WhichSocket:NiagaraInt32=&RandomRangeInt(Minimum=0, Maximum=2)) {isSetParameters=true, module=SetVariables_#1}
   &InitializeParticle(
    Lifetime=&Multiply_Float(A=1, B=&MakeCustomFloatFromBool(Bool=$User.IsOnGround, True Float=1, False Float=0))
    Mesh Scale=&Multiply_VectorByFloat(
     Vector=&Multiply_VectorByFloat(
      Vector=<1 1 1>
      Float=&Multiply_Float(
       A=0.1
       B=&SelectFloatFromArray(
        Direct Array Index=$Particles.WhichSocket
        Float Selection Array=DataInterface() {propertyValues {FloatData [1,0.7,1.2,1.1,0.8,1.3,1.3,1.25]}}
       )
      )
     )
     Float=&RandomRangeFloat(Minimum=1, Maximum=1.3)
    )
    Module.Write Color=true
    Module.Write Lifetime=true
    Module.Write Mass=true
    Module.Write Material Random=true
    Module.Write Position=false
    Module.Write Scale=true
    Module.Write SpriteRotation=false
    Module.Write SpriteSize=true
    Position=&SimulationPosition()
    Sprite Size=<4 10>
   )
   &SkeletalMeshLocation(
    Bone / Socket Index=$Particles.WhichSocket
    Bone Sampling Mode=::NewEnumerator4
    Inherit Velocity=false
    MaskSpawnGroup=false
    Sampled Position Offset=&Add_Vector(
     A=&SelectVectorFromArray(
      Direct Array Index=$Particles.WhichSocket
      Vector Selection Array=DataInterface() {propertyValues {FloatData [<0 0 0>, <-15 0 0>, <15 0 0>]}}
     )
     B=&RandomVector(Vector Scale=&SelectFloatFromArray(Direct Array Index=$Particles.WhichSocket, Float Selection Array=DataInterface() {propertyValues {FloatData [12,4,4]}}))
    )
    Sampled Position Offset Coordinate Space:ENiagaraMeshTransforms=::NewEnumerator3
    Skeletal Mesh=$User.SkeletalMesh
   )
   &ApplyInitialForces()
   &InheritVelocity(Module.Limit Speed=true, Velocity Scale=&Multiply_VectorByFloat(Vector=<1 1 1>, Float=-1)) {module=InheritVelocity001}
   /SetVariables(Particles.SocketFloat:NiagaraFloat=&MakeFloatFromInt(Int=$Particles.WhichSocket)) {isSetParameters=true}
  ]
  particleUpdateScript [
   &ParticleState()
   &GravityForce(Gravity=&Multiply_VectorByFloat(
     Vector=<1 1 -700>
     Float=&FloatFromCurve(FloatCurve=DataInterface() {propertyValues {Curve {keys [{arriveTangent=1.8039, interpMode=RCIM_Cubic, leaveTangent=1.8039, tangentMode=RCTM_User, time=0, value=0}, <at 1 1>]}}})
    ))
   &SolveForcesAndVelocity(Clamp Velocity=false, Limit Acceleration=false)
   &ScaleSpriteSize(Scale Factor=&Vector2DFromCurve(
     Vector2Curve=DataInterface() {propertyValues {
       XCurve {keys [
         <at 0 1.15 RCIM_Cubic>
         {arriveTangent=-0.8581, interpMode=RCIM_Cubic, leaveTangent=-0.8581, time=0.6037, value=1.14}
         {arriveTangent=-0.3533, interpMode=RCIM_Cubic, leaveTangent=-0.3533, time=0.8741, value=0.4}
         <at 1 1>
        ]}
       YCurve {keys [
         <at 0 0 RCIM_Cubic>
         {arriveTangent=2.8602, interpMode=RCIM_Cubic, leaveTangent=2.8602, time=0.6037, value=0.9885}
         {arriveTangent=2.0477, interpMode=RCIM_Cubic, leaveTangent=2.0477, time=0.8741, value=2.5}
         <at 1 1.8>
        ]}
      }}
     ScaleCurve=<1 1>
    ))
   &DynamicMaterialParameters(
    Index 0 Param 1=0.8
    Index 0 Param 2=&FloatFromCurve(FloatCurve=DataInterface() {propertyValues {Curve {keys [{arriveTangent=-1.8884, interpMode=RCIM_Cubic, leaveTangent=-1.8884, tangentMode=RCTM_User, time=0, value=1}, <at 1 0.26>]}}})
    Index 0 Param 3=1
    Index 0 Param 4=1
    Module.Param0WriteEnabled=true
    Module.Param1WriteEnabled=true
    Module.Param2WriteEnabled=true
    Module.Param3WriteEnabled=true
   )
   &CameraOffset(Camera Offset Amount=&SelectFloatFromArray(Direct Array Index=$Particles.WhichSocket, Float Selection Array=DataInterface() {propertyValues {FloatData [15,15,15]}}))
  ]
  rendererClasses [&NiagaraSpriteRendererProperties]
  renderers [{rendererClass=&NiagaraSpriteRendererProperties}]
 }
}
summary {
 systemName=NS_ChickenWarrior_Slurp_Idle
 emitters [
  {emitterName=Jump_Drips, rendererClasses [&NiagaraSpriteRendererProperties]}
  {emitterName=Idle_Drips, rendererClasses [&NiagaraSpriteRendererProperties]}
 ]
 userVariables [
  User.bIsFrontEnd {type=NiagaraBool}
  User.bIsFrontEndPreview {type=NiagaraBool}
  User.IsOnGround {type=NiagaraBool}
  User.IsSkydiving {type=NiagaraBool}
  User.JumpSpawn {type=NiagaraFloat, defaultValue=0, description=""}
  User.Land {type=NiagaraFloat, defaultValue=0, description=""}
  User.runningSpeed {type=NiagaraFloat, defaultValue=0, description=""}
  User.SkeletalMesh {
   type=NiagaraDataInterfaceSkeletalMesh
   defaultValue {
    struct=NiagaraExt_VariableValue_DataInterface
    value {dataInterface=&self:NiagaraDataInterfaceSkeletalMesh_0}
   }
  }
 ]
}
