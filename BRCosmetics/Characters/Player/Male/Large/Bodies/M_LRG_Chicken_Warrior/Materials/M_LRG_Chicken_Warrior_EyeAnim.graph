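# Character material for the large Chicken Warrior body that swaps eye and eyebrow shapes to match facial poses.
# Pose scalars are thresholded by rounding to pick target values, which VertexColorMask compares against red (eyes) and green (brows) vertex colors.
# World Position Offset collapses vertices outside the combined mask, which includes a head green value, to the object position.
# It wraps this between shared character PreFX and PostFX functions, adding specular and eye/eyebrow diffuse multipliers.
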Material /BRCosmetics/Characters/Player/Male/Large/Bodies/M_LRG_Chicken_Warrior/Materials/M_LRG_Chicken_Warrior_EyeAnim
graph {
 properties {
  bUseMaterialAttributes=true
  bUsedWithMeshDeformer=true
  bUsedWithMorphTargets=true
  bUsedWithSkeletalMesh=true
  pixelDepthOffsetMode=PDOM_Legacy
  usedShadingModels=Default Lit
 }
 nodes [
  n1 = VertexColor()
  ChannelMaskParameter(None=@n1) {group=Banana, parameterName=EyeVertexColorChannel Red}
  n2 = MaterialFunctionCall() {materialFunction=&BaseFN_CharacterFunction_PreFX}
  n3 = GetMaterialAttributes(None=@n2.Result) {attributeGetTypes [#1, #2, #3]}
  n4 = WorldPosition()
  n5 = ObjectPositionWS()
  n6 = VertexColor()
  n7 = Constant() {r=1}
  n8 = ScalarParameter() {defaultValue=0.3, group=Banana, parameterName="Eye (Default)"}
  n9 = ScalarParameter() {defaultValue=0.5, group=Banana, parameterName=Eye Squint}
  n10 = Max(A=ScalarParameter() {group=Banana_Poses, parameterName=R_squint_inner_pose}) {constB=0}
  n11 = Divide(B=ScalarParameter() {defaultValue=0.1, parameterName=TH_squint}) {constA=0.5}
  n12 = Clamp(None=Round(None=Multiply(A=@n10, B=@n11)))
  n13 = Multiply(
   A=ScalarParameter() {group=Banana, parameterName=C_glabella_up_pose}
   B=Divide(B=ScalarParameter() {defaultValue=2, parameterName=TH_glabella_up}) {constA=0.5}
  )
  n14 = OneMinus(None=Clamp(None=Round(None=@n13)))
  n15 = Multiply(
   A=ScalarParameter() {group=Banana, parameterName=R_brow_up_pose}
   B=Divide(B=ScalarParameter() {defaultValue=0.3, parameterName=TH_brow_up}) {constA=0.5}
  )
  n16 = Multiply(
   A=ScalarParameter() {group=Banana_Poses, parameterName=C_glabella_down_pose}
   B=Divide(B=ScalarParameter() {defaultValue=0.2, parameterName=TH_glabella_down}) {constA=0.5}
  )
  n17 = Max(A=Clamp(None=Round(None=@n15)), B=Clamp(None=Round(None=@n16)))
  n18 = Multiply(
   A=Max(A=ScalarParameter() {group=Banana_Poses, parameterName=R_blink_pose}) {constB=0}
   B=Divide(B=ScalarParameter() {defaultValue=0.3, parameterName=TH_narrow}) {constA=0.5}
  )
  n19 = LinearInterpolate(
   A=LinearInterpolate(
    A=LinearInterpolate(A=@n8, B=@n9, Alpha=@n12)
    B=ScalarParameter() {defaultValue=0.3, group=Banana, parameterName="Eye (Default)"}
    Alpha=Min(A=@n14, B=@n17)
   )
   B=ScalarParameter() {defaultValue=0.6, group=Banana, parameterName=Eye Narrow}
   Alpha=Clamp(None=Round(None=@n18))
  )
  n20 = Constant() {r=1}
  n21 = ScalarParameter() {defaultValue=0.05, group=Banana, parameterName=Brow Up}
  n22 = Max(A=ScalarParameter() {group=Banana, parameterName=R_brow_up_pose}) {constB=0}
  n23 = ScalarParameter() {defaultValue=0.1, group=Banana, parameterName=TH_Brow}
  n24 = Divide(B=@n23) {constA=0.5}
  n25 = Clamp(None=Round(None=Multiply(A=@n22, B=@n24)))
  n26 = ScalarParameter() {group=Banana, parameterName=C_glabella_down_pose}
  n27 = Round(None=Multiply(A=Max(A=@n26) {constB=0}, B=Divide(B=@n23) {constA=0.5}))
  n28 = Multiply(A=Max(A=ScalarParameter() {group=Banana, parameterName=C_glabella_up_pose}) {constB=0}, B=Divide(B=@n23) {constA=0.5})
  n29 = LinearInterpolate(
   A=LinearInterpolate(
    A=LinearInterpolate(A=@n21, B=@n21, Alpha=@n25)
    B=ScalarParameter() {defaultValue=0.2, group=Banana, parameterName="Glabella Down (Default)"}
    Alpha=Clamp(None=@n27)
   )
   B=ScalarParameter() {defaultValue=0.1, group=Banana, parameterName=Glabella Up}
   Alpha=Clamp(None=Round(None=@n28))
  )
  n30 = Max(
   A=MaterialFunctionCall!(
    Hide Element Toggle=Constant() {r=1}
    VertexColorChannel=ChannelMaskParameter(None=@n6) {group=Banana, maskChannel=Green, parameterName=EyeBrowVertexColorChannel Green}.R
    Vertex Value to hide=ScalarParameter() {defaultValue=1, parameterName=HeadGreenValue}
   ) {materialFunction=/Game/Packages/Fortress_SharedMaterials/Base_Material_Functions/VertexColorMask}.Result
   B=Max(
    A=MaterialFunctionCall(Hide Element Toggle=@n7, VertexColorChannel=@n1.R, Vertex Value to hide=@n19) {materialFunction=&VertexColorMask}.Result
    B=MaterialFunctionCall(Hide Element Toggle=@n20, VertexColorChannel=@n6.G, Vertex Value to hide=@n29) {materialFunction=&VertexColorMask}.Result
   )
  )
  n31 = Multiply(A=Multiply(A=Subtract(A=@n4.XYZ, B=@n5)) {constB=-1}, B=Saturate(None=OneMinus(None=@n30)))
  n32 = ScalarParameter() {defaultValue=1, group=Specular, parameterName=SpecularMultiplier, sortPriority=100}
  n33 = LinearInterpolate(
   A=ScalarParameter() {defaultValue=1, group=Diffuse, parameterName=EyesDiffuseMultiplier, sortPriority=100}
   B=ScalarParameter() {defaultValue=1, group=Diffuse, parameterName=EyebrowsDiffuseMultiplier, sortPriority=110}
   Alpha=Ceil(None=VertexColor().G)
  )
  n34 = MaterialFunctionCall(
   AO=@n2.AO
   Distance Blend=@n2.AO
   Fresnel=@n2.AO
   Gmap=@n2.AO
   Pre Skinned Local Position=@n2.AO
   Material=SetMaterialAttributes!(
    MaterialAttributes=@n3.MaterialAttributes
    World Position Offset=Add(A=@n3.World Position Offset, B=@n31)
    Specular=Multiply(A=@n3.Specular, B=@n32)
    Base Color=Multiply(A=@n3.Base Color, B=@n33)
   ) {attributeSetTypes [#1, #2, #3]}
  ) {materialFunction=&BaseFN_CharacterFunction_PostFX}
  MaterialFunctionCall(In=@n34.Result) {materialFunction=&CINE_MF_CharPostFX}
  ScalarParameter() {defaultValue=0.2, group=Banana, parameterName="Glabella Down (Default)"}
 ]
}
