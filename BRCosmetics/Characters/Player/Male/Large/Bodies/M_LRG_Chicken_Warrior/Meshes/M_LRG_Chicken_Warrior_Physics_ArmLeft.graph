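# Physics asset for the left arm of the large male Chicken Warrior player body cosmetic, covering the left front strap bones.
# It holds two box bodies: a small one on the strap master bone and a longer, flat, offset one on the first strap bone.
# A single constraint joins the two, with moderate swing limits and a tighter twist limit.
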PhysicsAsset /BRCosmetics/Characters/Player/Male/Large/Bodies/M_LRG_Chicken_Warrior/Meshes/M_LRG_Chicken_Warrior_Physics_ArmLeft
bodies {
 dyn_strapFront_01_l [{shapeType=Box, extentX=6.8983, extentY=2.8292, extentZ=1.2047, center=<2.725 0.0118 -0.1259>}]
 dyn_strapFront_mstr_l [{shapeType=Box, extentX=1.01, extentY=1.01, extentZ=1.01}]
}
constraints [{bone1Name=dyn_strapFront_01_l, bone2Name=dyn_strapFront_mstr_l, swing1LimitDegrees=45, swing2LimitDegrees=45, twistLimitDegrees=15}]
