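# Physics asset for the right-side front strap on the large male Chicken Warrior player body cosmetic, named as its right-arm physics setup.
# It has two box bodies: a long, thin one on the strap bone and a small cube on the strap's master bone.
# A single constraint joins the two bones and defines swing and twist limit angles for the strap's motion relative to the master.
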PhysicsAsset /BRCosmetics/Characters/Player/Male/Large/Bodies/M_LRG_Chicken_Warrior/Meshes/M_LRG_Chicken_Warrior_Physics_ArmRight
bodies {
 dyn_strapFront_01_r [{shapeType=Box, extentX=7.1721, extentY=2.7895, extentZ=1.0796, center=<2.9897 0.0404 -0.0811>}]
 dyn_strapFront_mstr_r [{shapeType=Box, extentX=1.01, extentY=1.01, extentZ=1.01}]
}
constraints [{bone1Name=dyn_strapFront_01_r, bone2Name=dyn_strapFront_mstr_r, swing1LimitDegrees=45, swing2LimitDegrees=45, twistLimitDegrees=15}]
