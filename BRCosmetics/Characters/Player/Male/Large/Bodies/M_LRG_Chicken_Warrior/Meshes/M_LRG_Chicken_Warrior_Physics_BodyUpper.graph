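# Physics asset for the upper body of the large male Chicken Warrior player cosmetic, covering its left and right back straps.
# Each side has a small box body on a master strap bone and a larger, flattened box on the first strap segment.
# A constraint on each side links the segment to its master with limited swing and tighter twist, with no link between the two sides.
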PhysicsAsset /BRCosmetics/Characters/Player/Male/Large/Bodies/M_LRG_Chicken_Warrior/Meshes/M_LRG_Chicken_Warrior_Physics_BodyUpper
bodies {
 dyn_strapBack_01_l [{shapeType=Box, extentX=4.7283, extentY=3.9147, extentZ=1.8587, center=<2.23 0.121 -0.1083>}]
 dyn_strapBack_01_r [{shapeType=Box, extentX=5.2074, extentY=3.9372, extentZ=1.6589, center=<2.1266 -0.1271 -0.0986>}]
 dyn_strapBack_mstr_l [{shapeType=Box, extentX=1.01, extentY=1.01, extentZ=1.01}]
 dyn_strapBack_mstr_r [{shapeType=Box, extentX=1.01, extentY=1.01, extentZ=1.01}]
}
constraints [
 {bone1Name=dyn_strapBack_01_r, bone2Name=dyn_strapBack_mstr_r, swing1LimitDegrees=45, swing2LimitDegrees=45, twistLimitDegrees=15}
 {bone1Name=dyn_strapBack_01_l, bone2Name=dyn_strapBack_mstr_l, swing1LimitDegrees=45, swing2LimitDegrees=45, twistLimitDegrees=15}
]
