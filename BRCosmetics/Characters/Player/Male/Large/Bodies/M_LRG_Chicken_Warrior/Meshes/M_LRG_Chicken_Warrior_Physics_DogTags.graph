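# Physics asset for the dog tag necklace on the large male Chicken Warrior character body.
# Box bodies form a chain from the necklace master bone through three necklace bones, then branch into left and right two-bone dog tags.
# Narrow swing and twist limits constrain the chain, while its master joins the upper spine with free rotation.
# Two large spine capsules provide torso volumes.
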PhysicsAsset /BRCosmetics/Characters/Player/Male/Large/Bodies/M_LRG_Chicken_Warrior/Meshes/M_LRG_Chicken_Warrior_Physics_DogTags
bodies {
 dyn_dogtag_01_l [{shapeType=Box, extentX=0.62, extentY=2.55, extentZ=2.6046, center=<0.9907 0.0117 -0.114>, rotation=<rot -4.7609 -90 -16.7551>}]
 dyn_dogtag_01_r [{shapeType=Box, extentX=0.6027, extentY=2.64, extentZ=2.6807, center=<1.019 0.001 0.1462>, rotation=<rot 0 90.0543 -81.396>}]
 dyn_dogtag_02_l [{shapeType=Box, extentX=4, extentY=0.5, extentZ=7.5, center=<2.9029 0.0281 -0.0535>, rotation=<rot 0 89.1397 -89.1873>}]
 dyn_dogtag_02_r [{shapeType=Box, extentX=4, extentY=0.5, extentZ=7.5, center=<2.8654 0.0766 -0.0165>, rotation=<rot 0 88.8758 -89.7396>}]
 dyn_necklace_01 [{shapeType=Box, extentX=4.0979, extentY=0.9789, extentZ=16.3881, center=<1 0 0>, rotation=<rot 0 0 90>}]
 dyn_necklace_02 [{shapeType=Box, extentX=4.704, extentY=0.9396, extentZ=12.327, center=<1 0 0>, rotation=<rot 0 0 90>}]
 dyn_necklace_03 [{shapeType=Box, extentX=2.9588, extentY=1.0097, extentZ=5.5456, center=<1 0 0>, rotation=<rot 0 0 90>}]
 dyn_necklace_mstr [{shapeType=Box, extentX=1.01, extentY=1.01, extentZ=1.01, rotation=<rot 0 -90 0>}]
 spine_04 [{shapeType=Capsule, length=13.8606, radius=20.9928}]
 spine_05 [{shapeType=Capsule, length=14.9065, radius=18.898, center=<-3.6297 0.8351 0>}]
}
constraints [
 {bone1Name=dyn_dogtag_02_r, bone2Name=dyn_dogtag_01_r, swing1LimitDegrees=4, swing2LimitDegrees=25, twistLimitDegrees=5}
 {bone1Name=dyn_necklace_01, bone2Name=dyn_necklace_mstr, swing1LimitDegrees=4, swing2LimitDegrees=20, twistLimitDegrees=5}
 {bone1Name=dyn_necklace_02, bone2Name=dyn_necklace_01, swing1LimitDegrees=4, swing2LimitDegrees=20, twistLimitDegrees=5}
 {bone1Name=dyn_necklace_03, bone2Name=dyn_necklace_02, swing1LimitDegrees=4, swing2LimitDegrees=25, twistLimitDegrees=5}
 {bone1Name=dyn_dogtag_01_l, bone2Name=dyn_necklace_03, swing1LimitDegrees=4, swing2LimitDegrees=25, twistLimitDegrees=5}
 {bone1Name=dyn_dogtag_02_l, bone2Name=dyn_dogtag_01_l, swing1LimitDegrees=4, swing2LimitDegrees=25, twistLimitDegrees=5}
 {bone1Name=dyn_dogtag_01_r, bone2Name=dyn_necklace_03, swing1LimitDegrees=4, swing2LimitDegrees=25, twistLimitDegrees=5}
 {
  bone1Name=dyn_necklace_mstr
  bone2Name=spine_05
  swing1LimitDegrees=45
  swing1Motion=Free
  swing2LimitDegrees=45
  swing2Motion=Free
  twistLimitDegrees=45
  twistMotion=Free
 }
 {
  bone1Name=spine_05
  bone2Name=spine_04
  swing1LimitDegrees=45
  swing1Motion=Free
  swing2LimitDegrees=45
  swing2Motion=Free
  twistLimitDegrees=45
  twistMotion=Free
 }
]
