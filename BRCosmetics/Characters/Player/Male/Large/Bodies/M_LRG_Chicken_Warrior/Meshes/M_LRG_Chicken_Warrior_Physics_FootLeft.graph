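# Physics asset for the left shoe tongue of the M_LRG_Chicken_Warrior cosmetic character body.
# It defines two box bodies, a small one on the master tongue bone and a larger, offset one on the tongue bone.
# A single constraint joins the two bones with tight swing and twist limits, allowing only slight movement of the tongue.
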PhysicsAsset /BRCosmetics/Characters/Player/Male/Large/Bodies/M_LRG_Chicken_Warrior/Meshes/M_LRG_Chicken_Warrior_Physics_FootLeft
bodies {
 dyn_shoeTongue_01_l [{shapeType=Box, extentX=8.7273, extentY=11.2775, extentZ=4.0959, center=<3.5575 0.2386 0.7356>}]
 dyn_shoeTongue_mstr_l [{shapeType=Box, extentX=1.01, extentY=1.01, extentZ=1.01}]
}
constraints [{bone1Name=dyn_shoeTongue_01_l, bone2Name=dyn_shoeTongue_mstr_l, swing1LimitDegrees=5, swing2LimitDegrees=5, twistLimitDegrees=2}]
