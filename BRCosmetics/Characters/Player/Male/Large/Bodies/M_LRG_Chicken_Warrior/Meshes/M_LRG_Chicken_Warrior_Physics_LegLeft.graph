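# Physics asset for secondary motion on the left leg of the M_LRG_Chicken_Warrior player body: four eggs and upper and lower thigh straps.
# Each dynamic bone has a box body constrained to its own small "mstr" box body, forming six independent single-link pairs.
# Tight angular limits confine each piece mostly to one swing axis, with eggs swinging modestly on swing1 and straps swinging wider on swing2.
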PhysicsAsset /BRCosmetics/Characters/Player/Male/Large/Bodies/M_LRG_Chicken_Warrior/Meshes/M_LRG_Chicken_Warrior_Physics_LegLeft
bodies {
 dyn_egg_01 [{shapeType=Box, extentX=5.9773, extentY=6.1259, extentZ=7.2619, center=<2.7641 0.4229 -0.2733>}]
 dyn_egg_01_mstr [{shapeType=Box, extentX=1.01, extentY=1.01, extentZ=1.01}]
 dyn_egg_02 [{shapeType=Box, extentX=6.0042, extentY=6.1204, extentZ=7.2996, center=<2.8113 0.1012 -0.25>}]
 dyn_egg_02_mstr [{shapeType=Box, extentX=1.01, extentY=1.01, extentZ=1.01}]
 dyn_egg_03 [{shapeType=Box, extentX=5.9344, extentY=6.1067, extentZ=7.2646, center=<2.9378 0.1462 0.0951>}]
 dyn_egg_03_mstr [{shapeType=Box, extentX=1.01, extentY=1.01, extentZ=1.01}]
 dyn_egg_04 [{shapeType=Box, extentX=5.8416, extentY=6.185, extentZ=7.2843, center=<2.8881 -0.3241 0.1771>}]
 dyn_egg_04_mstr [{shapeType=Box, extentX=1.01, extentY=1.01, extentZ=1.01}]
 dyn_strapThighLower_01 [{shapeType=Box, extentX=5.474, extentY=3.1725, extentZ=0.9816, center=<2.1939 0.0432 -0.0133>}]
 dyn_strapThighLower_mstr [{shapeType=Box, extentX=1.01, extentY=1.01, extentZ=1.01}]
 dyn_strapThighUpper_01 [{shapeType=Box, extentX=3.2077, extentY=3.519, extentZ=0.804, center=<1.4654 -0.0006 0.0378>}]
 dyn_strapThighUpper_mstr [{shapeType=Box, extentX=1.01, extentY=1.01, extentZ=1.01}]
}
constraints [
 {bone1Name=dyn_egg_01, bone2Name=dyn_egg_01_mstr, swing1LimitDegrees=15, swing2LimitDegrees=2, twistLimitDegrees=2}
 {bone1Name=dyn_egg_02, bone2Name=dyn_egg_02_mstr, swing1LimitDegrees=15, swing2LimitDegrees=2, twistLimitDegrees=2}
 {bone1Name=dyn_egg_03, bone2Name=dyn_egg_03_mstr, swing1LimitDegrees=15, swing2LimitDegrees=2, twistLimitDegrees=2}
 {bone1Name=dyn_egg_04, bone2Name=dyn_egg_04_mstr, swing1LimitDegrees=15, swing2LimitDegrees=2, twistLimitDegrees=2}
 {bone1Name=dyn_strapThighUpper_01, bone2Name=dyn_strapThighUpper_mstr, swing1LimitDegrees=2, swing2LimitDegrees=45, twistLimitDegrees=2}
 {bone1Name=dyn_strapThighLower_01, bone2Name=dyn_strapThighLower_mstr, swing1LimitDegrees=2, swing2LimitDegrees=45, twistLimitDegrees=2}
]
