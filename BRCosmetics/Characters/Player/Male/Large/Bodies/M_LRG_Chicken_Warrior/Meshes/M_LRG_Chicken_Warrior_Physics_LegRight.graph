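# Physics asset for the right leg of the large male Chicken Warrior player body, covering only the shoe tongue's dynamic bones.
# A box body on the tongue bone is constrained to a small box on its master bone, which looks like the anchor.
# The constraint has tight swing and twist limits, so the tongue can wobble only slightly.
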PhysicsAsset /BRCosmetics/Characters/Player/Male/Large/Bodies/M_LRG_Chicken_Warrior/Meshes/M_LRG_Chicken_Warrior_Physics_LegRight
bodies {
 dyn_shoeTongue_01_r [{shapeType=Box, extentX=9.2142, extentY=11.2775, extentZ=4.3045, center=<3.3165 -0.2386 0.6324>}]
 dyn_shoeTongue_mstr_r [{shapeType=Box, extentX=1.01, extentY=1.01, extentZ=1.01}]
}
constraints [{bone1Name=dyn_shoeTongue_01_r, bone2Name=dyn_shoeTongue_mstr_r, swing1LimitDegrees=5, swing2LimitDegrees=5, twistLimitDegrees=2}]
