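# Animation Blueprint for the Chicken Warrior large male cosmetic body, adding secondary motion to the pose copied from the attached parent mesh.
# Weighted constraints position four strap bones, then chained bone-space RigidBody nodes simulate arms, upper body, legs, left foot and dog tags.
# An additive necklace-bone offset applies when bRigidBodyEnabled is false.
# Above material quality 0, each update sends root-bone axes to the dynamic material and refreshes that offset's alpha.
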AnimBlueprint /BRCosmetics/Characters/Player/Male/Large/Bodies/M_LRG_Chicken_Warrior/Meshes/M_LRG_Chicken_Warrior_AnimBP
target_skeleton=&/M_LRG_Chicken_Warrior_Skeleton
graphs [
 AnimGraph {nodes [
   n1 = "Constraint - dyn_strapBack_mstr_r (PPPPPP)"!(
    ComponentPose=Local To Component(LocalPose=Copy Pose From Mesh() {bUseAttachedParent=true})
    Alpha=1
    ConstraintWeights_0=0.1016
    ConstraintWeights_1:float=0.3689
    ConstraintWeights_2:float=0.0554
    ConstraintWeights_3:float=0.0125
    ConstraintWeights_4:float=0.362
    ConstraintWeights_5:float=0.0996
   ) -> (Pose) {
    alphaBoolBlend {blendOption=Linear}
    boneToModify {boneName=dyn_strapBack_mstr_r}
    constraintSetup [
     {targetBone {boneName=spine_02}}
     {targetBone {boneName=spine_03}}
     {targetBone {boneName=spine_04}}
     {targetBone {boneName=spine_05}}
     {targetBone {boneName=clavicle_r}}
     {targetBone {boneName=pelvis}}
    ]
    constraintWeights [1,1,1,1,1,1]
   }
   n2 = "Constraint - dyn_strapBack_mstr_l (PPPPPP)"!(
    ComponentPose=@n1
    Alpha=1
    ConstraintWeights_0=0.3031
    ConstraintWeights_1:float=0.1102
    ConstraintWeights_2:float=0.4059
    ConstraintWeights_3:float=0.0609
    ConstraintWeights_4:float=0.0077
    ConstraintWeights_5:float=0.1123
   ) -> (Pose) {
    alphaBoolBlend {blendOption=Linear}
    boneToModify {boneName=dyn_strapBack_mstr_l}
    constraintSetup [
     {targetBone {boneName=clavicle_l}}
     {targetBone {boneName=spine_02}}
     {targetBone {boneName=spine_03}}
     {targetBone {boneName=spine_04}}
     {targetBone {boneName=spine_05}}
     {targetBone {boneName=pelvis}}
    ]
    constraintWeights [1,1,1,1,1,1]
   }
   n3 = "Constraint - dyn_strapFront_mstr_l (PPPPPPP)"!(
    ComponentPose=@n2
    Alpha=1
    ConstraintWeights_0=0.6043
    ConstraintWeights_1:float=0.0439
    ConstraintWeights_2:float=0.0694
    ConstraintWeights_3:float=0.0599
    ConstraintWeights_4:float=0.0637
    ConstraintWeights_5:float=0.1413
    ConstraintWeights_6:float=0.0175
   ) -> (Pose) {
    alphaBoolBlend {blendOption=Linear}
    boneToModify {boneName=dyn_strapFront_mstr_l}
    constraintSetup [
     {targetBone {boneName=clavicle_l}}
     {targetBone {boneName=upperarm_twist_02_l}}
     {targetBone {boneName=spine_03}}
     {targetBone {boneName=spine_04}}
     {targetBone {boneName=spine_05}}
     {targetBone {boneName=upperarm_twist_01_l}}
     {targetBone {boneName=upperarm_l}}
    ]
    constraintWeights [1,1,1,1,1,1,1]
   }
   n4 = "Constraint - dyn_strapFront_mstr_r (PPPPPPP)"!(
    ComponentPose=@n3
    Alpha=1
    ConstraintWeights_0=0.0148
    ConstraintWeights_1:float=0.143
    ConstraintWeights_2:float=0.044
    ConstraintWeights_3:float=0.0695
    ConstraintWeights_4:float=0.0591
    ConstraintWeights_5:float=0.0597
    ConstraintWeights_6:float=0.6099
   ) -> (Pose) {
    alphaBoolBlend {blendOption=Linear}
    boneToModify {boneName=dyn_strapFront_mstr_r}
    constraintSetup [
     {targetBone {boneName=upperarm_r}}
     {targetBone {boneName=upperarm_twist_01_r}}
     {targetBone {boneName=upperarm_twist_02_r}}
     {targetBone {boneName=spine_03}}
     {targetBone {boneName=spine_04}}
     {targetBone {boneName=spine_05}}
     {targetBone {boneName=clavicle_r}}
    ]
    constraintWeights [1,1,1,1,1,1,1]
   }
   n5 = RigidBody!(ComponentPose=@n4, Alpha, OverrideWorldGravity:Vector="0,0,-980", ExternalForce) -> (Pose) {
    bClampLinearTranslationLimitToRefPose=true
    bOverrideWorldGravity=true
    overridePhysicsAsset=&/M_LRG_Chicken_Warrior_Physics_ArmLeft
    simulationSpace=BaseBoneSpace
    alphaBoolBlend {blendOption=Linear}
    baseBoneRef {boneName=upperarm_l}
    componentAppliedLinearAccClamp=<3000 3000 3000>
    componentLinearAccScale=<0.4 0.4 0.4>
    componentLinearVelScale=<0.4 0.4 0.4>
    overrideWorldGravity {z=-980}
    simSpaceSettings {
     maxAngularAcceleration=1000
     maxAngularVelocity=3
     maxLinearAcceleration=700
     maxLinearVelocity=500
     velocityScaleZ=0
     worldAlpha=1
     externalLinearDragV=<1 1>
    }
   }
   n6 = RigidBody(ComponentPose=@n5, OverrideWorldGravity="0,0,-980") {
    bClampLinearTranslationLimitToRefPose=true
    bOverrideWorldGravity=true
    overridePhysicsAsset=&/M_LRG_Chicken_Warrior_Physics_ArmRight
    simulationSpace=BaseBoneSpace
    alphaBoolBlend {blendOption=Linear}
    baseBoneRef {boneName=upperarm_r}
    componentAppliedLinearAccClamp=<3000 3000 3000>
    componentLinearAccScale=<0.4 0.4 0.4>
    componentLinearVelScale=<0.4 0.4 0.4>
    overrideWorldGravity {z=-980}
    simSpaceSettings {
     maxAngularAcceleration=1000
     maxAngularVelocity=3
     maxLinearAcceleration=700
     maxLinearVelocity=500
     velocityScaleZ=0
     worldAlpha=1
     externalLinearDragV=<1 1>
    }
   }
   n7 = RigidBody(ComponentPose=@n6, OverrideWorldGravity="0,0,-980") {
    bClampLinearTranslationLimitToRefPose=true
    bOverrideWorldGravity=true
    overridePhysicsAsset=&/M_LRG_Chicken_Warrior_Physics_BodyUpper
    simulationSpace=BaseBoneSpace
    alphaBoolBlend {blendOption=Linear}
    baseBoneRef {boneName=spine_04}
    componentAppliedLinearAccClamp=<3000 3000 3000>
    componentLinearAccScale=<0.4 0.4 0.4>
    componentLinearVelScale=<0.4 0.4 0.4>
    overrideWorldGravity {z=-980}
    simSpaceSettings {
     maxAngularAcceleration=1000
     maxAngularVelocity=3
     maxLinearAcceleration=700
     maxLinearVelocity=500
     velocityScaleZ=0
     worldAlpha=1
     externalLinearDragV=<1 1>
    }
   }
   n8 = RigidBody(ComponentPose=@n7, OverrideWorldGravity="0,0,-980") {
    bClampLinearTranslationLimitToRefPose=true
    bOverrideWorldGravity=true
    overridePhysicsAsset=&/M_LRG_Chicken_Warrior_Physics_LegLeft
    simulationSpace=BaseBoneSpace
    alphaBoolBlend {blendOption=Linear}
    baseBoneRef {boneName=thigh_l}
    componentAppliedLinearAccClamp=<3000 3000 3000>
    componentLinearAccScale=<0.4 0.4 0.4>
    componentLinearVelScale=<0.4 0.4 0.4>
    overrideWorldGravity {z=-980}
    simSpaceSettings {
     maxAngularAcceleration=1000
     maxAngularVelocity=3
     maxLinearAcceleration=700
     maxLinearVelocity=500
     velocityScaleZ=0
     worldAlpha=1
     externalLinearDragV=<1 1>
    }
   }
   n9 = RigidBody(ComponentPose=@n8, OverrideWorldGravity="0,0,-980") {
    bClampLinearTranslationLimitToRefPose=true
    bOverrideWorldGravity=true
    overridePhysicsAsset=&/M_LRG_Chicken_Warrior_Physics_FootLeft
    simulationSpace=BaseBoneSpace
    alphaBoolBlend {blendOption=Linear}
    baseBoneRef {boneName=calf_l}
    componentAppliedLinearAccClamp=<3000 3000 3000>
    componentLinearAccScale=<0.4 0.4 0.4>
    componentLinearVelScale=<0.4 0.4 0.4>
    overrideWorldGravity {z=-980}
    simSpaceSettings {
     maxAngularAcceleration=1000
     maxAngularVelocity=3
     maxLinearAcceleration=700
     maxLinearVelocity=500
     velocityScaleZ=0
     worldAlpha=1
     externalLinearDragV=<1 1>
    }
   }
   n10 = RigidBody(ComponentPose=@n9, OverrideWorldGravity="0,0,-980") {
    bClampLinearTranslationLimitToRefPose=true
    bOverrideWorldGravity=true
    overridePhysicsAsset=&/M_LRG_Chicken_Warrior_Physics_LegRight
    simulationSpace=BaseBoneSpace
    alphaBoolBlend {blendOption=Linear}
    baseBoneRef {boneName=calf_r}
    componentAppliedLinearAccClamp=<3000 3000 3000>
    componentLinearAccScale=<0.4 0.4 0.4>
    componentLinearVelScale=<0.4 0.4 0.4>
    overrideWorldGravity {z=-980}
    simSpaceSettings {
     maxAngularAcceleration=1000
     maxAngularVelocity=3
     maxLinearAcceleration=700
     maxLinearVelocity=500
     velocityScaleZ=0
     worldAlpha=1
     externalLinearDragV=<1 1>
    }
   }
   n11 = RigidBody(ComponentPose=@n10, OverrideWorldGravity="0,0,-980") {
    bClampLinearTranslationLimitToRefPose=true
    bOverrideWorldGravity=true
    overridePhysicsAsset=&/M_LRG_Chicken_Warrior_Physics_DogTags
    simulationSpace=BaseBoneSpace
    alphaBoolBlend {blendOption=Linear}
    baseBoneRef {boneName=pelvis}
    componentAppliedLinearAccClamp=<2000 2000 2000>
    componentLinearAccScale=<0.2 0.2 0.2>
    componentLinearVelScale=<0.2 0.2 0.2>
    overrideWorldGravity {z=-980}
    simSpaceSettings {
     maxAngularAcceleration=1000
     maxAngularVelocity=3
     maxLinearAcceleration=700
     maxLinearVelocity=500
     velocityScaleZ=0
     worldAlpha=1
     externalLinearDragV=<1 1>
    }
   }
   n12 = "Transform (Modify) Bone - Bone: dyn_necklace_03"(
    ComponentPose=@n11
    Alpha=Get rigidBodyEnabledFloat() -> (rigidBodyEnabledFloat:double)
    Translation="0,1,0"
    Rotation="0,0,-20"
    Scale="1,1,1"
   ) {
    rotationMode=BMM_Additive
    translationMode=BMM_Additive
    alphaBoolBlend {blendOption=Linear}
    boneToModify {boneName=dyn_necklace_03}
    rotation {roll=-20}
    translation {y=1}
   }
   Output Pose(Result=Component To Local(ComponentPose=@n12)) {layerGroup=None}
  ]}
 EventGraph {nodes [
   Event BlueprintUpdateAnimation()
   > Branch(Condition="integer > integer"(A=GetRenderingMaterialQualityLevel!() -> (ReturnValue:int=0), B=0) {is=title K2Node_CallFunction})
   n1 = GetPartSkeletalMeshComponent()
   > n2 = Get Dynamic Material Instance(SkeletalMeshComponent=@n1)
   n3 = TransformToBoneSpace!(self:Skinned Mesh Component Object Reference=@n1, BoneName:Name=root, InPosition:Vector="0, 0, 0", InRotation:Rotator="0, 0, 0") -> (OutPosition:Vector, OutRotation:Rotator)
   n4 = MakeTransform(Location=@n3, Rotation=@n3.OutRotation)
   > SetVectorParameterValue(
    self=@n2
    ParameterName=ActorRotationX
    Value="To LinearColor (Vector)"!(InVec:Vector=TransformDirection!(T:"Transform (by ref)"=@n4, Direction:Vector="1,0, 0") -> (ReturnValue:Vector)) -> (ReturnValue:Linear Color Structure)
   )
   > SetVectorParameterValue(self=@n2, ParameterName=ActorRotationY, Value="To LinearColor (Vector)"(InVec=TransformDirection(T=@n4, Direction="0,1, 0")))
   > SetVectorParameterValue(self=@n2, ParameterName=ActorRotationZ, Value="To LinearColor (Vector)"(InVec=TransformDirection(T=@n4, Direction="0, 0,1")))
   > n5 = Branch(Condition=Get bRigidBodyEnabled())
   > Set rigidBodyEnabledFloat(rigidBodyEnabledFloat:double=0) -> (Output_Get:double)
   Set rigidBodyEnabledFloat(execute=@n5.else, rigidBodyEnabledFloat=1)
  ]}
]
variables [rigidBodyEnabledFloat {type=number}]
