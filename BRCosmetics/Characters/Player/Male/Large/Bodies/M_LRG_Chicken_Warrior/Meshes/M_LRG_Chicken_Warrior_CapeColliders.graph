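# Physics asset of cape colliders for the large male Chicken Warrior player body, approximating the body as simple shapes for the cape to collide against.
# Capsules cover the pelvis, two spine bones, upper arms, thighs and calves, while spheres cover the upper neck and head.
# Forearms, hands and feet have no bodies, and there are no constraints, so the asset holds collision shapes only with no ragdoll joints.
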PhysicsAsset /BRCosmetics/Characters/Player/Male/Large/Bodies/M_LRG_Chicken_Warrior/Meshes/M_LRG_Chicken_Warrior_CapeColliders
bodies {
 calf_l [{shapeType=Capsule, length=31, radius=12.75, center=<-25.6313 0.0001 0.6181>, rotation=<rot 88.6184 0.0001 0.0001>}]
 calf_r [{shapeType=Capsule, length=32.5, radius=12.25, center=<25.4165 0 -0.9123>, rotation=<rot 87.9443 0 0>}]
 head [{shapeType=Sphere, radius=19.5, center=<16.5572 0 0>}]
 neck_02 [{shapeType=Sphere, radius=15}]
 pelvis [{shapeType=Capsule, length=20.5, radius=15.75, center=<-3.1525 -1.924 0>}]
 spine_03 [{shapeType=Capsule, length=16.5, radius=20.75, center=<-1.7119 -1.4343 -1.2546>, rotation=<rot 88.2308 -0.0002 6.518>}]
 spine_05 [{shapeType=Capsule, length=13, radius=25.25, center=<-3.1075 3.3059 0.0001>}]
 thigh_l [{shapeType=Capsule, length=25, radius=13.75, center=<-27.0964 -0.0002 0.9335>, rotation=<rot 88.0269 0 0>}]
 thigh_r [{shapeType=Capsule, length=25.75, radius=13, center=<26.0752 0.0003 0.1355>, rotation=<rot 88.0269 0 0>}]
 upperarm_l [{shapeType=Capsule, length=16.5, radius=13.5, center=<5.4609 -0.1101 -0.7677>, rotation=<rot 89.4503 -179.999 171.1026>}]
 upperarm_r [{shapeType=Capsule, length=16, radius=14, center=<-4.6811 0.6605 -0.0515>, rotation=<rot 89.4872 -179.9988 173.9>}]
}
constraints []
