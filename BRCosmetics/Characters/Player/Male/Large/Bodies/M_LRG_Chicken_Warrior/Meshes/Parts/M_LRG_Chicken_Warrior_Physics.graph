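# Physics asset for the Chicken Warrior large male player cosmetic, covering the head and the left and right helmet straps.
# Each strap is a chain of four box bodies hanging from a small master box, linked in series with tight swing and twist limits that favor one swing axis.
# Both master bodies constrain to a single head capsule with free swing and twist.
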PhysicsAsset /BRCosmetics/Characters/Player/Male/Large/Bodies/M_LRG_Chicken_Warrior/Meshes/Parts/M_LRG_Chicken_Warrior_Physics
bodies {
 dyn_strapHelmet_01_l [{shapeType=Box, extentX=4, extentY=1.5, extentZ=1.5, center=<2 0 0>}]
 dyn_strapHelmet_01_r [{shapeType=Box, extentX=4, extentY=1.5, extentZ=1.5, center=<2 0 0>}]
 dyn_strapHelmet_02_l [{shapeType=Box, extentX=4, extentY=1.5, extentZ=1.5, center=<2 0 0>}]
 dyn_strapHelmet_02_r [{shapeType=Box, extentX=4, extentY=1.5, extentZ=1.5, center=<2 0 0>}]
 dyn_strapHelmet_03_l [{shapeType=Box, extentX=4, extentY=1.5, extentZ=1.5, center=<2 0 0>}]
 dyn_strapHelmet_03_r [{shapeType=Box, extentX=4, extentY=1.5, extentZ=1.5, center=<2 0 0>}]
 dyn_strapHelmet_04_l [{shapeType=Box, extentX=4, extentY=1.5, extentZ=1.5, center=<2 0 0>}]
 dyn_strapHelmet_04_r [{shapeType=Box, extentX=4, extentY=1.5, extentZ=1.5, center=<2 0 0>}]
 dyn_strapHelmet_mstr_l [{shapeType=Box, extentX=1.01, extentY=1.01, extentZ=1.01}]
 dyn_strapHelmet_mstr_r [{shapeType=Box, extentX=1.01, extentY=1.01, extentZ=1.01}]
 head [{shapeType=Capsule, length=31.1933, radius=11.8316, center=<-3.6613 -6.8347 0>, rotation=<rot 4.8521 89.9996 0>}]
}
constraints [
 {bone1Name=dyn_strapHelmet_01_r, bone2Name=dyn_strapHelmet_mstr_r, swing1LimitDegrees=15, swing2LimitDegrees=5, twistLimitDegrees=5}
 {bone1Name=dyn_strapHelmet_02_r, bone2Name=dyn_strapHelmet_01_r, swing1LimitDegrees=15, swing2LimitDegrees=5, twistLimitDegrees=5}
 {bone1Name=dyn_strapHelmet_03_r, bone2Name=dyn_strapHelmet_02_r, swing1LimitDegrees=15, swing2LimitDegrees=5, twistLimitDegrees=5}
 {bone1Name=dyn_strapHelmet_04_r, bone2Name=dyn_strapHelmet_03_r, swing1LimitDegrees=15, swing2LimitDegrees=5, twistLimitDegrees=5}
 {bone1Name=dyn_strapHelmet_01_l, bone2Name=dyn_strapHelmet_mstr_l, swing1LimitDegrees=15, swing2LimitDegrees=5, twistLimitDegrees=5}
 {bone1Name=dyn_strapHelmet_02_l, bone2Name=dyn_strapHelmet_01_l, swing1LimitDegrees=15, swing2LimitDegrees=5, twistLimitDegrees=5}
 {bone1Name=dyn_strapHelmet_03_l, bone2Name=dyn_strapHelmet_02_l, swing1LimitDegrees=15, swing2LimitDegrees=5, twistLimitDegrees=5}
 {bone1Name=dyn_strapHelmet_04_l, bone2Name=dyn_strapHelmet_03_l, swing1LimitDegrees=15, swing2LimitDegrees=5, twistLimitDegrees=5}
 {
  bone1Name=dyn_strapHelmet_mstr_r
  bone2Name=head
  swing1LimitDegrees=45
  swing1Motion=Free
  swing2LimitDegrees=45
  swing2Motion=Free
  twistLimitDegrees=45
  twistMotion=Free
 }
 {
  bone1Name=dyn_strapHelmet_mstr_l
  bone2Name=head
  swing1LimitDegrees=45
  swing1Motion=Free
  swing2LimitDegrees=45
  swing2Motion=Free
  twistLimitDegrees=45
  twistMotion=Free
 }
]
